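# Cow moo one-shot ambient cue, played quietly in the Fort_Ambient_Props_Athena sound class.
# A quality-level node picks between equally weighted random pools: all eight Cow waves on the first input, only Cow04 to Cow06 on the second, with the other two inputs empty.
# Overridden attenuation takes the place of the referenced huge-prop preset, adding natural-sound falloff, distance low-pass, occlusion, listener focus and submix sends.
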SoundCue /Game/Sounds/Fort_Ambient_NonLoop/Animals/Fort_CowMoo_Cue
root=QualityLevel_0
nodes [
 {id=QualityLevel_0, children [Random_2,Random_0,null,null]}
 {id=Random_0, children [WavePlayer_9, WavePlayer_10, WavePlayer_13], properties {weights [1,1,1]}}
 {
  id=Random_2
  children [WavePlayer_5, WavePlayer_6, WavePlayer_7, WavePlayer_3, WavePlayer_4, WavePlayer_8, WavePlayer_11, WavePlayer_12]
  properties {weights [1,1,1,1,1,1,1,1]}
 }
 {id=WavePlayer_10, properties {soundWaveAssetPtr=&/Cow05}}
 {id=WavePlayer_11, properties {soundWaveAssetPtr=&/Cow07}}
 {id=WavePlayer_12, properties {soundWaveAssetPtr=&/Cow08}}
 {id=WavePlayer_13, properties {soundWaveAssetPtr=&/Cow06}}
 {id=WavePlayer_3, properties {soundWaveAssetPtr=&/Cow04}}
 {id=WavePlayer_4, properties {soundWaveAssetPtr=&Cow05}}
 {id=WavePlayer_5, properties {soundWaveAssetPtr=&/Cow01}}
 {id=WavePlayer_6, properties {soundWaveAssetPtr=&/Cow02}}
 {id=WavePlayer_7, properties {soundWaveAssetPtr=&/Cow03}}
 {id=WavePlayer_8, properties {soundWaveAssetPtr=&Cow06}}
 {id=WavePlayer_9, properties {soundWaveAssetPtr=&Cow04}}
]
properties {
 attenuationSettings=/Game/Sounds/Attenuation/Fort_Amb_Emitter_Prop_Huge_Att
 bOverrideAttenuation=true
 maxDistance=5000
 soundClassObject=&Fort_Ambient_Props_Athena
 volumeMultiplier=0.3
 attenuationOverrides {
  bAttenuateWithLPF=true
  bEnableListenerFocus=true
  bEnableOcclusion=true
  bEnableSubmixSends=true
  distanceAlgorithm=NaturalSound
  falloffDistance=3000
  lPFFrequencyAtMin=1000
  nonFocusVolumeAttenuation=0.7
  occlusionLowPassFilterFrequency=6000
  occlusionTraceChannel=ECC_GameTraceChannel11
  reverbDistanceMax=6000
  attenuationShapeExtents {x=2000}
 }
}
