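# Sound cue for a rubber duck being hit, layering a short-delayed squeak over a rubber impact in the impacts sound class.
# A quality switch gives its first level a mix of randomly chosen duck and rubber waves, each layer pitch-modulated.
# The other three levels share a lighter branch with one wave per layer and pitch variation.
# The cue overrides attenuation with logarithmic distance falloff.
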SoundCue /Game/Sounds/Fort_Ambient_NonLoop/Animals/SC_RubberDuckHit_Cue
root=QualityLevel_0
nodes [
 {id=Delay_0, children [Modulator_3], properties {delayMax=0.1, delayMin=0.08}}
 {id=Delay_1, children [WavePlayer_9], properties {delayMax=0.08, delayMin=0.08}}
 {id=Mixer_0, children [Delay_0,Modulator_0], properties {inputVolume [1.5,0.75]}}
 {id=Mixer_1, children [Delay_1,Modulator_2], properties {inputVolume [1,0.5]}}
 {id=Modulator_0, children [Random_0], properties {pitchMax=1.5, pitchMin=1.2}}
 {id=Modulator_1, children [Mixer_1], properties {pitchMax=1.25, pitchMin=0.85}}
 {id=Modulator_2, children [WavePlayer_0], properties {pitchMax=1.5, pitchMin=1.2}}
 {id=Modulator_3, children [Random_2], properties {pitchMax=1.2, pitchMin=0.7}}
 {id=QualityLevel_0, children [Mixer_0,Modulator_1,Modulator_1,Modulator_1]}
 {id=Random_0, children [WavePlayer_91, WavePlayer_92, WavePlayer_93], properties {weights [1,1,1]}}
 {id=Random_2, children [WavePlayer_2,WavePlayer_3,WavePlayer_4,WavePlayer_5], properties {weights [1,1,1,1]}}
 {id=WavePlayer_0, properties {soundWaveAssetPtr=/Game/Sounds/Fort_Impact_Sounds/Demolition/Rubber/Rubber_Impact01}}
 {id=WavePlayer_2, properties {soundWaveAssetPtr=&/RubberDuckHit_01}}
 {id=WavePlayer_3, properties {soundWaveAssetPtr=&/RubberDuckHit_02}}
 {id=WavePlayer_4, properties {soundWaveAssetPtr=&/RubberDuckHit_03}}
 {id=WavePlayer_5, properties {soundWaveAssetPtr=&/RubberDuckHit_04}}
 {id=WavePlayer_9, properties {soundWaveAssetPtr=&RubberDuckHit_01}}
 {id=WavePlayer_91, properties {soundWaveAssetPtr=&Rubber_Impact01}}
 {id=WavePlayer_92, properties {soundWaveAssetPtr=/Game/Sounds/Fort_Impact_Sounds/Demolition/Rubber/Rubber_Impact02}}
 {id=WavePlayer_93, properties {soundWaveAssetPtr=/Game/Sounds/Fort_Impact_Sounds/Demolition/Rubber/Rubber_Impact03}}
]
properties {
 bOverrideAttenuation=true
 maxDistance=2500
 soundClassObject=&Fort_Impacts
 volumeMultiplier=0.8
 attenuationOverrides {distanceAlgorithm=Logarithmic, falloffDistance=1500, attenuationShapeExtents {x=1000}}
}
